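# Physics asset for the Desert Ops Swamp female face accessory, defining collision and joints for its dyn_hair bones.
# Spheres and capsules cover the clavicles, head, neck and upper spine, while boxes cover the dynamic bones.
# A free-rotating dyn_hair_bk_all body links the head to a tightly limited back master, which carries five child bodies with angular limits.
# The other four bodies (03–06) attach directly to the head with free swing.
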PhysicsAsset /BRCosmetics/Accessories/Hats/F_MED_Desert_Ops_Swamp_FaceAcc/Meshes/F_MED_Desert_Ops_Swamp_FaceAcc_PhysicsAsset
bodies {
 clavicle_l [{shapeType=Sphere, radius=5.213, center=<7 0 0>}]
 clavicle_r [{shapeType=Sphere, radius=8.531, center=<-7.4918 -0.8764 2.0193>}]
 dyn_hair_01 [{shapeType=Box, extentX=0.5, extentY=5.6897, extentZ=9.6031, center=<4.2564 0 0>, rotation=<rot 90 0 0>}]
 dyn_hair_02 [{shapeType=Box, extentX=0.8675, extentY=8, extentZ=15.028, center=<7.8033 0 0.0002>, rotation=<rot 90 0 0>}]
 dyn_hair_03 [{shapeType=Box, extentX=2, extentY=8, extentZ=4, center=<2 0 0>, rotation=<rot 90 0 0>}]
 dyn_hair_04 [{shapeType=Box, extentX=2, extentY=8, extentZ=4, center=<2 0 0>, rotation=<rot 90 0 0>}]
 dyn_hair_05 [{shapeType=Box, extentX=2, extentY=8, extentZ=7.6925, center=<4.1781 0 0>, rotation=<rot 90 0 0>}]
 dyn_hair_06 [{shapeType=Box, extentX=2, extentY=4.7951, extentZ=13.5714, center=<6.2027 0 0>, rotation=<rot 90 0 0>}]
 dyn_hair_07 [{shapeType=Box, extentX=0.5282, extentY=8, extentZ=13.8487, center=<6.6644 0 0>, rotation=<rot 90 0 0>}]
 dyn_hair_08 [{shapeType=Box, extentX=0.6675, extentY=14.6764, extentZ=13.2079, center=<5.9177 0 0.0001>, rotation=<rot 90 0 0>}]
 dyn_hair_09 [{shapeType=Box, extentX=0.6955, extentY=16.7531, extentZ=15.7396, center=<8.894 -0.0001 0.0001>, rotation=<rot 90 0 0>}]
 dyn_hair_bk_all [{shapeType=Box, extentX=1.01, extentY=1.01, extentZ=1.01}]
 dyn_hair_bk_mstr [{shapeType=Box, extentX=2.3939, extentY=19.7838, extentZ=21.9193, center=<11.6939 0 -0.0001>, rotation=<rot 90 0 0>}]
 head [{shapeType=Capsule, length=5.1278, radius=5.4343, center=<-0.4858 1.7503 -0.191>, rotation=<rot 0 117.6878 -89.7355>}]
 neck_01 [{shapeType=Capsule, length=4.9125, radius=6.4663, center=<0.4349 0 0>, rotation=<rot 90 0 0>}]
 spine_05 [{shapeType=Capsule, length=27.8683, radius=14.2495, center=<-4.2309 0.1804 0>}]
}
constraints [
 {bone1Name=dyn_hair_08, bone2Name=dyn_hair_bk_mstr, swing1LimitDegrees=35, swing2LimitDegrees=40, twistLimitDegrees=5}
 {bone1Name=dyn_hair_07, bone2Name=dyn_hair_bk_mstr, swing1LimitDegrees=35, swing2LimitDegrees=40, twistLimitDegrees=5}
 {
  bone1Name=dyn_hair_06
  bone2Name=head
  swing1LimitDegrees=45
  swing1Motion=Free
  swing2LimitDegrees=45
  swing2Motion=Free
  twistLimitDegrees=5
 }
 {
  bone1Name=dyn_hair_05
  bone2Name=head
  swing1LimitDegrees=45
  swing1Motion=Free
  swing2LimitDegrees=45
  swing2Motion=Free
  twistLimitDegrees=5
 }
 {
  bone1Name=dyn_hair_04
  bone2Name=head
  swing1LimitDegrees=45
  swing1Motion=Free
  swing2LimitDegrees=45
  swing2Motion=Free
  twistLimitDegrees=5
 }
 {
  bone1Name=dyn_hair_03
  bone2Name=head
  swing1LimitDegrees=45
  swing1Motion=Free
  swing2LimitDegrees=45
  swing2Motion=Free
  twistLimitDegrees=5
 }
 {bone1Name=dyn_hair_02, bone2Name=dyn_hair_bk_mstr, swing1LimitDegrees=35, swing2LimitDegrees=65, twistLimitDegrees=5}
 {bone1Name=dyn_hair_01, bone2Name=dyn_hair_bk_mstr, swing1LimitDegrees=35, swing2LimitDegrees=40, twistLimitDegrees=5}
 {bone1Name=dyn_hair_09, bone2Name=dyn_hair_bk_mstr, swing1LimitDegrees=35, swing2LimitDegrees=40, twistLimitDegrees=5}
 {bone1Name=dyn_hair_bk_mstr, bone2Name=dyn_hair_bk_all, swing1LimitDegrees=35, swing2LimitDegrees=10, twistLimitDegrees=3}
 {
  bone1Name=dyn_hair_bk_all
  bone2Name=head
  swing1LimitDegrees=45
  swing1Motion=Free
  swing2LimitDegrees=45
  swing2Motion=Free
  twistLimitDegrees=45
  twistMotion=Free
 }
]
