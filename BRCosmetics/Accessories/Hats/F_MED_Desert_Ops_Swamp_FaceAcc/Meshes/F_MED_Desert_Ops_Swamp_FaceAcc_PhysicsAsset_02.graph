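# Physics asset for the Desert Ops Swamp face accessory in the hats cosmetics, giving its dyn_hair bones bodies and joints for physical motion.
# A back chain runs head, bk_all, bk_kine, bk_mstr, with strands 01, 02 and 07–09 branching from bk_mstr with free or locked twist.
# Strands 03–06 joint directly to the head with free swing.
# Unconstrained spine, neck and right-clavicle bodies, plus the head, provide body shapes, with no left clavicle.
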PhysicsAsset /BRCosmetics/Accessories/Hats/F_MED_Desert_Ops_Swamp_FaceAcc/Meshes/F_MED_Desert_Ops_Swamp_FaceAcc_PhysicsAsset_02
bodies {
 clavicle_r [{shapeType=Sphere, radius=5.9811, center=<-6.0374 0 -3.3607>}]
 dyn_hair_01 [{shapeType=Box, extentX=0.6184, extentY=4.7435, extentZ=7.6417, center=<3.0821 0 0.0001>, rotation=<rot 90 0 0>}]
 dyn_hair_02 [{shapeType=Capsule, length=8.568, radius=0.9606, center=<5.0299 0 0.0001>, rotation=<rot 90 0 0>}]
 dyn_hair_03 [{shapeType=Box, extentX=1.3244, extentY=4.3734, extentZ=4.9202, center=<2 0 0>, rotation=<rot 90 0 0>}]
 dyn_hair_04 [{shapeType=Box, extentX=2, extentY=8, extentZ=4, center=<2 0 0>, rotation=<rot 90 0 0>}]
 dyn_hair_05 [{shapeType=Box, extentX=2, extentY=2.2164, extentZ=7.6925, center=<4.1781 0 0>, rotation=<rot 90 0 0>}]
 dyn_hair_06 [{shapeType=Box, extentX=2, extentY=4.7951, extentZ=13.5714, center=<6.2027 0 0>, rotation=<rot 90 0 0>}]
 dyn_hair_07 [{shapeType=Box, extentX=14.5937, extentY=8.0789, extentZ=0.9172, center=<6.9518 0 -0.0002>}]
 dyn_hair_08 [{shapeType=Box, extentX=14.5937, extentY=13.1287, extentZ=0.9172, center=<6.9518 0 -0.0002>}]
 dyn_hair_09 [{shapeType=Box, extentX=14.5937, extentY=17.5465, extentZ=0.9661, center=<6.9518 0 -0.0002>}]
 dyn_hair_bk_all [{shapeType=Box, extentX=1.01, extentY=1.01, extentZ=1.01}]
 dyn_hair_bk_kine [{shapeType=Capsule, length=0.505, radius=0.505, rotation=<rot 0 -90 0>}]
 dyn_hair_bk_mstr [{shapeType=Box, extentX=1.1275, extentY=19.7838, extentZ=21.9193, center=<11.6939 0 -0.0001>, rotation=<rot 90 0 0>}]
 head [{shapeType=Capsule, length=5.1278, radius=5.7682, center=<-0.4858 1.7503 -0.191>, rotation=<rot 0 117.6878 -89.7355>}]
 neck_01 [{shapeType=Capsule, length=12.465, radius=4.1186, center=<0.4349 0 0>, rotation=<rot 90 0 0>}]
 spine_05 [{shapeType=Capsule, length=18.31, radius=12.3402, center=<3.1511 1.7766 0>}]
}
constraints [
 {bone1Name=dyn_hair_08, bone2Name=dyn_hair_bk_mstr, swing1LimitDegrees=30, swing2LimitDegrees=60, twistLimitDegrees=5, twistMotion=Locked}
 {bone1Name=dyn_hair_07, bone2Name=dyn_hair_bk_mstr, swing1LimitDegrees=30, swing2LimitDegrees=60, twistLimitDegrees=5, twistMotion=Free}
 {
  bone1Name=dyn_hair_06
  bone2Name=head
  swing1LimitDegrees=45
  swing1Motion=Free
  swing2LimitDegrees=45
  swing2Motion=Free
  twistLimitDegrees=5
 }
 {
  bone1Name=dyn_hair_05
  bone2Name=head
  swing1LimitDegrees=45
  swing1Motion=Free
  swing2LimitDegrees=45
  swing2Motion=Free
  twistLimitDegrees=5
 }
 {
  bone1Name=dyn_hair_04
  bone2Name=head
  swing1LimitDegrees=45
  swing1Motion=Free
  swing2LimitDegrees=45
  swing2Motion=Free
  twistLimitDegrees=5
 }
 {
  bone1Name=dyn_hair_03
  bone2Name=head
  swing1LimitDegrees=45
  swing1Motion=Free
  swing2LimitDegrees=45
  swing2Motion=Free
  twistLimitDegrees=5
 }
 {bone1Name=dyn_hair_02, bone2Name=dyn_hair_bk_mstr, swing1LimitDegrees=30, swing2LimitDegrees=60, twistLimitDegrees=5, twistMotion=Free}
 {bone1Name=dyn_hair_01, bone2Name=dyn_hair_bk_mstr, swing1LimitDegrees=30, swing2LimitDegrees=60, twistLimitDegrees=5, twistMotion=Free}
 {bone1Name=dyn_hair_09, bone2Name=dyn_hair_bk_mstr, swing1LimitDegrees=30, swing2LimitDegrees=60, twistLimitDegrees=5, twistMotion=Locked}
 {bone1Name=dyn_hair_bk_mstr, bone2Name=dyn_hair_bk_kine, swing1LimitDegrees=45, swing2LimitDegrees=45, twistLimitDegrees=5}
 {
  bone1Name=dyn_hair_bk_all
  bone2Name=head
  swing1LimitDegrees=45
  swing1Motion=Free
  swing2LimitDegrees=45
  swing2Motion=Free
  twistLimitDegrees=45
  twistMotion=Free
 }
 {
  bone1Name=dyn_hair_bk_kine
  bone2Name=dyn_hair_bk_all
  swing1LimitDegrees=45
  swing1Motion=Free
  swing2LimitDegrees=45
  swing2Motion=Free
  twistLimitDegrees=45
  twistMotion=Free
 }
]
