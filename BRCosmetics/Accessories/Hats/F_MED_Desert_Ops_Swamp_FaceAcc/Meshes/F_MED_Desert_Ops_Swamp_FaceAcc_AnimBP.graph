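# Animation Blueprint for the Desert Ops Swamp face accessory, adding secondary motion to hair, hat and mask bones.
# A pose copied from the attached parent mesh feeds Flap nodes that offset the hat rear and back hair from head rotation.
# AnimDynamics chains with spring and collision limits, plus RigidBody nodes, simulate dangling parts.
# Variable-weighted constraints and bone transforms adjust hair, including during targeting; several node chains appear unconnected or zero-weighted.
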AnimBlueprint /BRCosmetics/Accessories/Hats/F_MED_Desert_Ops_Swamp_FaceAcc/Meshes/F_MED_Desert_Ops_Swamp_FaceAcc_AnimBP
target_skeleton=&/F_MED_Desert_Ops_Swamp_FaceAcc_Skeleton
graphs [
 AnimGraph {nodes [
   Anim Dynamics - Bone: dyn_hair_r_03(Alpha=1, ExternalForce="0,0,0") {
    angularDampingOverride=0.99
    angularSpringConstant=1000
    bAngularSpring=true
    bGravityOverrideInSimSpace=true
    bOverrideAngularDamping=true
    bUsePlanarLimit=false
    bUseSphericalLimits=true
    gravityScale=0.35
    alphaBoolBlend {blendOption=Linear}
    boundBone {boneName=dyn_hair_r_03}
    componentAppliedLinearAccClamp=<70 70 70>
    componentLinearVelScale=<300 300 300>
    gravityOverride {z=-980}
    physicsBodyDefinitions [{
      collisionType=CustomSphere
      sphereCollisionRadius=1.5
      boundBone {boneName=dyn_hair_r_03}
      boxExtents=<10 30 30>
      constraintSetup {angularLimitsMax=<0 45 5>, angularLimitsMin=<0 -2 -30>, angularTarget=<0 0 0>}
      localJointOffset=<-1.5 0 0>
     }]
    sphericalLimits [
     {limitRadius=60, drivingBone {boneName=spine_05}, sphereLocalOffset=<-48 -6.5 0>}
     {limitRadius=9, drivingBone {boneName=spine_05}, sphereLocalOffset=<13 0 0>}
    ]
   }
   n1 = Get hair_alpha() -> (hair_alpha:double)
   n2 = Get jacket_ext_force() -> (jacket_ext_force:Vector)
   n3 = Anim Dynamics - Bone: dyn_hair_bk_01(Alpha=@n1, ExternalForce=@n2) {
    angularDampingOverride=0.99
    angularSpringConstant=1000
    bAngularSpring=true
    bGravityOverrideInSimSpace=true
    bOverrideAngularDamping=true
    bUsePlanarLimit=false
    bUseSphericalLimits=true
    gravityScale=0.35
    alphaBoolBlend {blendOption=Linear}
    boundBone {boneName=dyn_hair_bk_01}
    componentAppliedLinearAccClamp=<70 70 70>
    componentLinearVelScale=<300 300 300>
    gravityOverride {z=-980}
    physicsBodyDefinitions [{
      collisionType=CustomSphere
      sphereCollisionRadius=4.3
      boundBone {boneName=dyn_hair_bk_01}
      boxExtents=<5 30 30>
      constraintSetup {angularLimitsMax=<0 20 25>, angularLimitsMin=<0 -20 -65>, angularTarget=<0 0 0>}
      localJointOffset=<6.5 0 0>
     }]
    planarLimits [{drivingBone {boneName=neck_01}, planeTransform {location=<14 -5 6>, rotation=<rot 90 -90 0>}}]
    sphericalLimits [{limitRadius=70, drivingBone {boneName=spine_05}, sphereLocalOffset=<-30 50 0>}]
   }
   n4 = Anim Dynamics - Bone: dyn_hair_bk_02(ComponentPose=@n3, Alpha=@n1, ExternalForce=@n2) {
    angularDampingOverride=0.99
    angularSpringConstant=1000
    bAngularSpring=true
    bGravityOverrideInSimSpace=true
    bOverrideAngularDamping=true
    bUseSphericalLimits=true
    gravityScale=0.35
    alphaBoolBlend {blendOption=Linear}
    boundBone {boneName=dyn_hair_bk_02}
    componentAppliedLinearAccClamp=<70 70 70>
    componentLinearVelScale=<300 300 300>
    gravityOverride {z=-980}
    physicsBodyDefinitions [{
      collisionType=CustomSphere
      sphereCollisionRadius=4
      boundBone {boneName=dyn_hair_bk_02}
      boxExtents=<5 40 40>
      constraintSetup {angularLimitsMax=<0 20 7>, angularLimitsMin=<0 -20 -75>, angularTarget=<0 0 0>}
      localJointOffset=<10 0 0>
     }]
    planarLimits [
     {drivingBone {boneName=neck_01}, planeTransform {location=<7 -7 6>, rotation=<rot -90 -90 0>}}
     {drivingBone {boneName=spine_05}, planeTransform {location=<7 -11.25 6>, rotation=<rot -90 -36.5651 0>}}
    ]
    sphericalLimits [
     {limitRadius=45, drivingBone {boneName=spine_05}, sphereLocalOffset=<-25 18 0>}
     {limitRadius=8, drivingBone {boneName=neck_02}, sphereLocalOffset=<0 -1.5 0>}
    ]
   }
   n5 = Anim Dynamics - Bone: dyn_hair_bk_r_01(ComponentPose=@n4, Alpha=@n1, ExternalForce=@n2) {
    angularDampingOverride=0.99
    angularSpringConstant=1000
    bAngularSpring=true
    bGravityOverrideInSimSpace=true
    bOverrideAngularDamping=true
    bUsePlanarLimit=false
    gravityScale=0.35
    alphaBoolBlend {blendOption=Linear}
    boundBone {boneName=dyn_hair_bk_r_01}
    componentAppliedLinearAccClamp=<70 70 70>
    componentLinearVelScale=<300 300 300>
    gravityOverride {z=-980}
    physicsBodyDefinitions [{
      collisionType=CustomSphere
      sphereCollisionRadius=3
      boundBone {boneName=dyn_hair_bk_r_01}
      boxExtents=<5 30 30>
      constraintSetup {angularLimitsMax=<0 10 20>, angularLimitsMin=<0 -10 -65>, angularTarget=<0 0 0>}
      localJointOffset=<6.5 0 0>
     }]
    sphericalLimits [{limitRadius=7, drivingBone {boneName=neck_02}, sphereLocalOffset=<0 0 0>}]
   }
   n6 = Anim Dynamics - Bone: dyn_hair_bk_r_02(ComponentPose=@n5, Alpha=@n1, ExternalForce=@n2) {
    angularDampingOverride=0.99
    angularSpringConstant=1000
    bAngularSpring=true
    bGravityOverrideInSimSpace=true
    bOverrideAngularDamping=true
    bUseSphericalLimits=true
    gravityScale=0.35
    alphaBoolBlend {blendOption=Linear}
    boundBone {boneName=dyn_hair_bk_r_02}
    componentAppliedLinearAccClamp=<70 70 70>
    componentLinearVelScale=<300 300 300>
    gravityOverride {z=-980}
    physicsBodyDefinitions [{
      collisionType=CustomSphere
      sphereCollisionRadius=4
      boundBone {boneName=dyn_hair_bk_r_02}
      boxExtents=<5 40 40>
      constraintSetup {angularLimitsMax=<0 10 7>, angularLimitsMin=<0 -10 -75>, angularTarget=<0 0 0>}
      localJointOffset=<10 0 0>
     }]
    sphericalLimits [{limitRadius=55.125, drivingBone {boneName=spine_05}, sphereLocalOffset=<-40 23 8>}]
   }
   n7 = Anim Dynamics - Bone: dyn_hair_bk_l_01(ComponentPose=@n6, Alpha=@n1, ExternalForce=@n2) {
    angularDampingOverride=0.99
    angularSpringConstant=1000
    bAngularSpring=true
    bGravityOverrideInSimSpace=true
    bOverrideAngularDamping=true
    bUsePlanarLimit=false
    gravityScale=0.35
    alphaBoolBlend {blendOption=Linear}
    boundBone {boneName=dyn_hair_bk_l_01}
    componentAppliedLinearAccClamp=<70 70 70>
    componentLinearVelScale=<300 300 300>
    gravityOverride {z=-980}
    physicsBodyDefinitions [{
      collisionType=CustomSphere
      sphereCollisionRadius=3
      boundBone {boneName=dyn_hair_bk_l_01}
      boxExtents=<5 30 30>
      constraintSetup {angularLimitsMax=<0 10 20>, angularLimitsMin=<0 -10 -65>, angularTarget=<0 0 0>}
      localJointOffset=<6.5 0 0>
     }]
    sphericalLimits [{limitRadius=7.5, drivingBone {boneName=neck_02}, sphereLocalOffset=<0 0 0>}]
   }
   n8 = Anim Dynamics - Bone: dyn_hair_bk_l_02(ComponentPose=@n7, Alpha=@n1, ExternalForce=@n2) {
    angularDampingOverride=0.99
    angularSpringConstant=1000
    bAngularSpring=true
    bGravityOverrideInSimSpace=true
    bOverrideAngularDamping=true
    bUsePlanarLimit=false
    bUseSphericalLimits=true
    gravityScale=0.35
    alphaBoolBlend {blendOption=Linear}
    boundBone {boneName=dyn_hair_bk_l_02}
    componentAppliedLinearAccClamp=<70 70 70>
    componentLinearVelScale=<300 300 300>
    gravityOverride {z=-980}
    physicsBodyDefinitions [{
      collisionType=CustomSphere
      sphereCollisionRadius=4
      boundBone {boneName=dyn_hair_bk_l_02}
      boxExtents=<5 40 40>
      constraintSetup {angularLimitsMax=<0 10 7>, angularLimitsMin=<0 -10 -75>, angularTarget=<0 0 0>}
      localJointOffset=<10 0 0>
     }]
    sphericalLimits [{limitRadius=46, drivingBone {boneName=spine_05}, sphereLocalOffset=<-30 19 -8>}]
   }
   n9 = Anim Dynamics - Bone: dyn_hair_l_01(ComponentPose=@n8, Alpha=@n1, ExternalForce=@n2) {
    angularDampingOverride=0.99
    angularSpringConstant=1000
    bAngularSpring=true
    bGravityOverrideInSimSpace=true
    bOverrideAngularDamping=true
    bUsePlanarLimit=false
    gravityScale=0.35
    alphaBoolBlend {blendOption=Linear}
    boundBone {boneName=dyn_hair_l_01}
    componentAppliedLinearAccClamp=<70 70 70>
    componentLinearVelScale=<300 300 300>
    gravityOverride {z=-980}
    physicsBodyDefinitions [{
      collisionType=CustomSphere
      sphereCollisionRadius=3
      boundBone {boneName=dyn_hair_l_01}
      boxExtents=<5 30 30>
      constraintSetup {angularLimitsMax=<0 8 3>, angularLimitsMin=<0 -8 -11>, angularTarget=<0 0 0>}
      localJointOffset=<-6.5 0 0>
     }]
    sphericalLimits [
     {limitRadius=8, drivingBone {boneName=neck_02}, sphereLocalOffset=<0 0 0>}
     {limitRadius=60, drivingBone {boneName=spine_05}, sphereLocalOffset=<-42 5 0>}
    ]
   }
   n10 = Anim Dynamics - Bone: dyn_hair_l_02(ComponentPose=@n9, Alpha=@n1, ExternalForce=@n2) {
    angularDampingOverride=0.99
    angularSpringConstant=200
    bAngularSpring=true
    bGravityOverrideInSimSpace=true
    bOverrideAngularDamping=true
    bUseSphericalLimits=true
    gravityScale=0.35
    alphaBoolBlend {blendOption=Linear}
    boundBone {boneName=dyn_hair_l_02}
    componentAppliedLinearAccClamp=<70 70 70>
    componentLinearVelScale=<300 300 300>
    gravityOverride {z=-980}
    physicsBodyDefinitions [{
      collisionType=CustomSphere
      sphereCollisionRadius=2.25
      boundBone {boneName=dyn_hair_l_02}
      boxExtents=<10 30 30>
      constraintSetup {angularLimitsMax=<0 8 40>, angularLimitsMin=<0 -8 -15>, angularTarget=<0 0 0>}
      localJointOffset=<-6.5 0 0>
     }]
    planarLimits [{drivingBone {boneName=clavicle_l}, planeTransform {location=<0 -2 0>, rotation=<rot -90 90 0>}}]
    sphericalLimits [
     {limitRadius=59, drivingBone {boneName=spine_05}, sphereLocalOffset=<-48 -6.5 0>}
     {limitRadius=6, drivingBone {boneName=spine_05}, sphereLocalOffset=<13 0 0>}
    ]
   }
   n11 = Anim Dynamics - Bone: dyn_hair_l_03(ComponentPose=@n10, Alpha=@n1, ExternalForce=@n2) {
    angularDampingOverride=0.99
    angularSpringConstant=1000
    bAngularSpring=true
    bGravityOverrideInSimSpace=true
    bOverrideAngularDamping=true
    bUsePlanarLimit=false
    bUseSphericalLimits=true
    gravityScale=0.35
    alphaBoolBlend {blendOption=Linear}
    boundBone {boneName=dyn_hair_l_03}
    componentAppliedLinearAccClamp=<70 70 70>
    componentLinearVelScale=<300 300 300>
    gravityOverride {z=-980}
    physicsBodyDefinitions [{
      collisionType=CustomSphere
      sphereCollisionRadius=3
      boundBone {boneName=dyn_hair_l_03}
      boxExtents=<5 30 30>
      constraintSetup {angularLimitsMax=<0 10 45>, angularLimitsMin=<0 -10 -25>, angularTarget=<0 0 0>}
      localJointOffset=<-6.5 0 0>
     }]
    sphericalLimits [
     {limitRadius=60, drivingBone {boneName=spine_05}, sphereLocalOffset=<-48 -6.5 0>}
     {limitRadius=8, drivingBone {boneName=spine_05}, sphereLocalOffset=<13 0 0>}
    ]
   }
   n12 = Anim Dynamics - Bone: dyn_hair_r_01(ComponentPose=@n11, Alpha=@n1, ExternalForce=@n2) {
    angularDampingOverride=0.99
    angularSpringConstant=1000
    bAngularSpring=true
    bGravityOverrideInSimSpace=true
    bOverrideAngularDamping=true
    bUsePlanarLimit=false
    gravityScale=0.35
    alphaBoolBlend {blendOption=Linear}
    boundBone {boneName=dyn_hair_r_01}
    componentAppliedLinearAccClamp=<70 70 70>
    componentLinearVelScale=<300 300 300>
    gravityOverride {z=-980}
    physicsBodyDefinitions [{
      collisionType=CustomSphere
      sphereCollisionRadius=3
      boundBone {boneName=dyn_hair_r_01}
      boxExtents=<5 30 30>
      constraintSetup {angularLimitsMax=<0 20 17>, angularLimitsMin=<0 -2.5 -8>, angularTarget=<0 0 0>}
      localJointOffset=<-6.5 0 0>
     }]
    sphericalLimits [
     {limitRadius=8, drivingBone {boneName=neck_02}, sphereLocalOffset=<0 0 0>}
     {limitRadius=8, drivingBone {boneName=neck_01}, sphereLocalOffset=<5 10 0>}
    ]
   }
   Anim Dynamics - Bone: dyn_hair_r_02(ComponentPose=@n12, Alpha=@n1, ExternalForce=@n2) {
    angularDampingOverride=0.99
    angularSpringConstant=200
    bAngularSpring=true
    bGravityOverrideInSimSpace=true
    bOverrideAngularDamping=true
    bUseSphericalLimits=true
    gravityScale=0.35
    alphaBoolBlend {blendOption=Linear}
    boundBone {boneName=dyn_hair_r_02}
    componentAppliedLinearAccClamp=<70 70 70>
    componentLinearVelScale=<300 300 300>
    gravityOverride {z=-980}
    physicsBodyDefinitions [{
      collisionType=CustomSphere
      sphereCollisionRadius=2.25
      boundBone {boneName=dyn_hair_r_02}
      boxExtents=<10 30 30>
      constraintSetup {angularLimitsMax=<0 20 20>, angularLimitsMin=<0 -5 -20>, angularTarget=<0 0 0>}
      localJointOffset=<-6.5 0 0>
     }]
    planarLimits [{drivingBone {boneName=clavicle_l}, planeTransform {location=<0 -2 0>, rotation=<rot -90 90 0>}}]
    sphericalLimits [
     {limitRadius=60, drivingBone {boneName=spine_05}, sphereLocalOffset=<-48 -6.5 0>}
     {limitRadius=6.5, drivingBone {boneName=spine_05}, sphereLocalOffset=<13 0 0>}
    ]
   }
   AnimGraphNode_SubInput_0 = Input Pose()
   n13 = Layered blend per bone(BlendPoses_0=@AnimGraphNode_SubInput_0) {layerSetup [{branchFilters [
       {blendDepth=1, boneName=dyn_hair_01}
       {blendDepth=1, boneName=dyn_hair_02}
       {blendDepth=1, boneName=dyn_hair_03}
       {blendDepth=1, boneName=dyn_hair_04}
       {blendDepth=1, boneName=dyn_hair_05}
       {blendDepth=1, boneName=dyn_hair_06}
       {blendDepth=1, boneName=dyn_hair_07}
       {blendDepth=1, boneName=dyn_hair_08}
       {blendDepth=1, boneName=dyn_hair_bk_all}
      ]}]}
   n14 = "Transform (Modify) Bone - Bone: dyn_hair_bk_all"(
    ComponentPose=Local To Component(LocalPose=@n13)
    Alpha=Get isTargeting() -> (isTargeting:double)
    Translation="-5,0,5"
    Rotation="0,0,-25"
    Scale="1,1,1"
   ) {
    rotationMode=BMM_Additive
    rotationSpace=BCS_BoneSpace
    scaleMode=BMM_Replace
    scaleSpace=BCS_BoneSpace
    translationMode=BMM_Additive
    translationSpace=BCS_BoneSpace
    alphaBoolBlend {blendOption=Linear}
    alphaScaleBiasClamp {bClampResult=true, bInterpResult=true, bMapRange=true, interpSpeedDecreasing=14, interpSpeedIncreasing=14}
    boneToModify {boneName=dyn_hair_bk_all}
    rotation {roll=-25}
    translation {x=-5, z=5}
   }
   n15 = RigidBody(ComponentPose=@n14) {
    bClampLinearTranslationLimitToRefPose=true
    bOverrideWorldGravity=true
    overridePhysicsAsset=&/F_MED_Desert_Ops_Swamp_FaceAcc_PhysicsAsset_02
    simulationSpace=BaseBoneSpace
    alphaBoolBlend {blendOption=Linear}
    alphaScaleBiasClamp {bInterpResult=true}
    baseBoneRef {boneName=pelvis}
    componentAppliedLinearAccClamp=<2000 2000 2000>
    componentLinearAccScale=<0.7 0.7 0.7>
    componentLinearVelScale=<0.7 0.7 0.7>
    overrideWorldGravity {z=-980}
   }
   Component To Local(ComponentPose=@n15)
   n16 = Get Constraint_Alpha() -> (Constraint_Alpha:double)
   n17 = "Constraint - dyn_hair_09 (P)"(Alpha=@n16, ConstraintWeights_0=1) {
    alphaBoolBlend {blendOption=Linear}
    boneToModify {boneName=dyn_hair_09}
    constraintSetup [{targetBone {boneName=spine_05}}]
    constraintWeights [1]
   }
   n18 = "Constraint - dyn_hair_08 (P)"(ComponentPose=@n17, Alpha=@n16, ConstraintWeights_0=1) {
    alphaBoolBlend {blendOption=Linear}
    boneToModify {boneName=dyn_hair_08}
    constraintSetup [{targetBone {boneName=spine_05}}]
    constraintWeights [1]
   }
   n19 = "Constraint - dyn_hair_07 (P)"(ComponentPose=@n18, Alpha=0, ConstraintWeights_0=1) {
    alphaBoolBlend {blendOption=Linear}
    boneToModify {boneName=dyn_hair_07}
    constraintSetup [{targetBone {boneName=spine_05}}]
    constraintWeights [1]
   }
   n20 = "Constraint - dyn_hair_01 (P)"(ComponentPose=@n19, Alpha=@n16, ConstraintWeights_0=1) {
    alphaBoolBlend {blendOption=Linear}
    boneToModify {boneName=dyn_hair_01}
    constraintSetup [{targetBone {boneName=spine_05}}]
    constraintWeights [1]
   }
   n21 = "Constraint - dyn_hair_08 (P)"(ComponentPose=@n20, Alpha=@n16, ConstraintWeights_0=1) {
    alphaBoolBlend {blendOption=Linear}
    boneToModify {boneName=dyn_hair_08}
    constraintSetup [{targetBone {boneName=clavicle_r}}]
    constraintWeights [1]
   }
   n22 = "Constraint - dyn_hair_09 (P)"(ComponentPose=@n21, Alpha=0, ConstraintWeights_0=1) {
    alphaBoolBlend {blendOption=Linear}
    boneToModify {boneName=dyn_hair_09}
    constraintSetup [{targetBone {boneName=clavicle_r}}]
    constraintWeights [1]
   }
   n23 = "Constraint - dyn_hair_02 (P)"(ComponentPose=@n22, Alpha=@n16, ConstraintWeights_0=1) {
    alphaBoolBlend {blendOption=Linear}
    boneToModify {boneName=dyn_hair_02}
    constraintSetup [{targetBone {boneName=clavicle_l}}]
    constraintWeights [1]
   }
   n24 = "Constraint - dyn_hair_07 (P)"(ComponentPose=@n23, Alpha=@n16, ConstraintWeights_0=1) {
    alphaBoolBlend {blendOption=Linear}
    boneToModify {boneName=dyn_hair_07}
    constraintSetup [{targetBone {boneName=clavicle_l}}]
    constraintWeights [1]
   }
   n25 = "Constraint - dyn_hair_03 (P)"(ComponentPose=@n24, Alpha=@n16, ConstraintWeights_0=1) {
    alphaBoolBlend {blendOption=Linear}
    boneToModify {boneName=dyn_hair_03}
    constraintSetup [{targetBone {boneName=clavicle_l}}]
    constraintWeights [1]
   }
   "Constraint - dyn_hair_01 (P)"(ComponentPose=@n25, Alpha=0, ConstraintWeights_0=1) {
    alphaBoolBlend {blendOption=Linear}
    boneToModify {boneName=dyn_hair_01}
    constraintSetup [{targetBone {boneName=dyn_hair_09}}]
    constraintWeights [1]
   }
   n26 = Get ADSUpWalk() -> (ADSUpWalk:bool)
   n27 = "Constraint - dyn_hair_bk_kine (P)"!(ComponentPose, bAlphaBoolEnabled:bool=@n26, ConstraintWeights_0=1) -> (Pose) {
    alphaInputType=Bool
    alphaBoolBlend {blendOption=Linear}
    boneToModify {boneName=dyn_hair_bk_kine}
    constraintSetup [{targetBone {boneName=spine_05}}]
    constraintWeights [1]
   }
   n28 = "Constraint - dyn_hair_09 (P)"!(ComponentPose=@n27, bAlphaBoolEnabled:bool=@n26, ConstraintWeights_0=1) -> (Pose) {
    alphaInputType=Bool
    alphaBoolBlend {blendOption=Linear}
    boneToModify {boneName=dyn_hair_09}
    constraintSetup [{targetBone {boneName=spine_05}}]
    constraintWeights [1]
   }
   "Constraint - dyn_hair_09 (P)"(ComponentPose=@n28, bAlphaBoolEnabled=@n26, ConstraintWeights_0=1) {
    alphaInputType=Bool
    alphaBoolBlend {blendOption=Linear}
    boneToModify {boneName=dyn_hair_09}
    constraintSetup [{targetBone {boneName=spine_05}}]
    constraintWeights [1]
   }
   "Constraint - dyn_hair_07 (P)"(Alpha=@n16, ConstraintWeights_0=1) {
    alphaBoolBlend {blendOption=Linear}
    boneToModify {boneName=dyn_hair_07}
    constraintSetup [{targetBone {boneName=clavicle_r}}]
    constraintWeights [1]
   }
   n29 = "Transform (Modify) Bone - Bone: dyn_hair_bk_all"(Alpha=0, Translation="0,0,5.5", Rotation="-25,-10,60", Scale="0.5,0.5,0.5") {
    rotationMode=BMM_Additive
    rotationSpace=BCS_BoneSpace
    scaleMode=BMM_Replace
    scaleSpace=BCS_BoneSpace
    translationMode=BMM_Additive
    translationSpace=BCS_BoneSpace
    alphaBoolBlend {blendOption=Linear}
    alphaScaleBiasClamp {bClampResult=true, bInterpResult=true, bMapRange=true, interpSpeedDecreasing=14, interpSpeedIncreasing=14}
    boneToModify {boneName=dyn_hair_bk_all}
    rotation=<rot -25 -10 60>
    scale=<0.5 0.5 0.5>
    translation {z=5.5}
   }
   "Transform (Modify) Bone - Bone: dyn_hair_bk_all"(ComponentPose=@n29, Alpha=0, Translation="-3,0,6.5", Rotation="10,-10,0", Scale="0.6,0.6,0.6") {
    rotationSpace=BCS_BoneSpace
    scaleMode=BMM_Replace
    scaleSpace=BCS_BoneSpace
    translationMode=BMM_Additive
    translationSpace=BCS_BoneSpace
    alphaBoolBlend {blendOption=Linear}
    alphaScaleBiasClamp {bClampResult=true, bInterpResult=true, bMapRange=true, interpSpeedDecreasing=14, interpSpeedIncreasing=14}
    boneToModify {boneName=dyn_hair_bk_all}
    rotation {pitch=10, yaw=-10}
    scale=<0.6 0.6 0.6>
    translation {x=-3, z=6.5}
   }
   n30 = Anim Dynamics - Bone: dyn_hat_strandA(Alpha=0.3, ExternalForce="0,0,0") {
    angularDampingOverride=0.985
    angularSpringConstant=150
    bAngularSpring=true
    bGravityOverrideInSimSpace=true
    bLinearSpring=true
    bOverrideAngularDamping=true
    bOverrideLinearDamping=true
    gravityScale=0.35
    linearDampingOverride=0.99
    linearSpringConstant=120
    simulationSpace=RootRelative
    alphaBoolBlend {blendOption=Linear}
    boundBone {boneName=dyn_hat_strandA}
    componentAppliedLinearAccClamp=<500 500 500>
    componentLinearAccScale=<1 1 1>
    componentLinearVelScale=<50 50 50>
    physicsBodyDefinitions [{
      sphereCollisionRadius=0
      boundBone {boneName=dyn_hat_strandA}
      boxExtents=<30 30 30>
      constraintSetup {
       angularTargetAxis=AxisY
       angularLimitsMax=<0 0 0>
       angularLimitsMin=<0 0 0>
       angularTarget=<0 0 0>
       linearAxesMax=<5.3768 6.3298 5>
       linearAxesMin=<-4.6209 -6.2114 -6>
      }
      localJointOffset=<0 0 0>
     }]
   }
   n31 = Anim Dynamics - Bone: dyn_hat_strandB(ComponentPose=@n30, Alpha=0.45, ExternalForce="0,0,0") {
    angularDampingOverride=0.985
    angularSpringConstant=160
    bAngularSpring=true
    bGravityOverrideInSimSpace=true
    bLinearSpring=true
    bOverrideAngularDamping=true
    bOverrideLinearDamping=true
    gravityScale=0.3
    linearDampingOverride=0.99
    linearSpringConstant=185
    simulationSpace=RootRelative
    alphaBoolBlend {blendOption=Linear}
    boundBone {boneName=dyn_hat_strandB}
    componentAppliedLinearAccClamp=<500 500 500>
    componentLinearAccScale=<1 1 1>
    componentLinearVelScale=<20 20 20>
    physicsBodyDefinitions [{
      sphereCollisionRadius=0
      boundBone {boneName=dyn_hat_strandB}
      boxExtents=<20 20 20>
      constraintSetup {
       angularTargetAxis=AxisY
       angularLimitsMax=<0 0 0>
       angularLimitsMin=<0 0 0>
       angularTarget=<0 0 0>
       linearAxesMax=<2.6406 5.3414 4>
       linearAxesMin=<-3.0262 -4.6437 -3>
      }
      localJointOffset=<0 0 0>
     }]
   }
   n32 = Anim Dynamics - Bone: dyn_hat_leafA(ComponentPose=@n31, Alpha=0.35, ExternalForce="0,0,0") {
    angularDampingOverride=0.985
    angularSpringConstant=255
    bAngularSpring=true
    bGravityOverrideInSimSpace=true
    bLinearSpring=true
    bOverrideAngularDamping=true
    bOverrideLinearDamping=true
    gravityScale=0.15
    linearDampingOverride=0.99
    linearSpringConstant=200
    alphaBoolBlend {blendOption=Linear}
    boundBone {boneName=dyn_hat_leafA}
    componentAppliedLinearAccClamp=<1000 1000 1000>
    componentLinearVelScale=<100 100 100>
    physicsBodyDefinitions [{
      sphereCollisionRadius=0
      boundBone {boneName=dyn_hat_leafA}
      boxExtents=<30 30 30>
      constraintSetup {
       angularTargetAxis=AxisY
       angularLimitsMax=<0 0 0>
       angularLimitsMin=<0 0 0>
       angularTarget=<0 0 0>
       linearAxesMax=<2 2 0.5>
       linearAxesMin=<-1.5 0 -0.5>
      }
      localJointOffset=<0 0 0>
     }]
   }
   n33 = Anim Dynamics - Bone: dyn_hat_leafB(ComponentPose=@n32, Alpha=0.3, ExternalForce="0,0,0") {
    angularDampingOverride=0.985
    angularSpringConstant=190
    bAngularSpring=true
    bGravityOverrideInSimSpace=true
    bLinearSpring=true
    bOverrideAngularDamping=true
    bOverrideLinearDamping=true
    gravityScale=0.2
    linearDampingOverride=0.99
    linearSpringConstant=175
    alphaBoolBlend {blendOption=Linear}
    boundBone {boneName=dyn_hat_leafB}
    componentAppliedLinearAccClamp=<1000 1000 1000>
    componentLinearVelScale=<100 100 100>
    physicsBodyDefinitions [{
      sphereCollisionRadius=0
      boundBone {boneName=dyn_hat_leafB}
      boxExtents=<30 30 30>
      constraintSetup {
       angularTargetAxis=AxisY
       angularLimitsMax=<0 0 0>
       angularLimitsMin=<0 0 0>
       angularTarget=<0 0 0>
       linearAxesMax=<2 2 1.5>
       linearAxesMin=<-2 -1 -1.5>
      }
      localJointOffset=<0 0 0>
     }]
   }
   n34 = Anim Dynamics - Bone: dyn_hat_leafC(ComponentPose=@n33, Alpha=0.35, ExternalForce="0,0,0") {
    angularDampingOverride=0.985
    angularSpringConstant=200
    bAngularSpring=true
    bGravityOverrideInSimSpace=true
    bLinearSpring=true
    bOverrideAngularDamping=true
    bOverrideLinearDamping=true
    gravityScale=0.2
    linearDampingOverride=0.99
    linearSpringConstant=150
    alphaBoolBlend {blendOption=Linear}
    boundBone {boneName=dyn_hat_leafC}
    componentAppliedLinearAccClamp=<1000 1000 1000>
    componentLinearVelScale=<100 100 100>
    physicsBodyDefinitions [{
      sphereCollisionRadius=0
      boundBone {boneName=dyn_hat_leafC}
      boxExtents=<20 10 15>
      constraintSetup {
       angularTargetAxis=AxisY
       angularLimitsMax=<0 0 0>
       angularLimitsMin=<0 0 0>
       angularTarget=<0 0 0>
       linearAxesMax=<3 2 0.5>
       linearAxesMin=<-0.5 0 -0.5>
      }
      localJointOffset=<0 0 0>
     }]
   }
   n35 = Anim Dynamics - Bone: dyn_hat_leafD(ComponentPose=@n34, Alpha=0.3, ExternalForce="0,0,0") {
    angularDampingOverride=0.985
    angularSpringConstant=150
    bAngularSpring=true
    bGravityOverrideInSimSpace=true
    bLinearSpring=true
    bOverrideAngularDamping=true
    bOverrideLinearDamping=true
    gravityScale=0.15
    linearDampingOverride=0.99
    linearSpringConstant=170
    alphaBoolBlend {blendOption=Linear}
    boundBone {boneName=dyn_hat_leafD}
    componentAppliedLinearAccClamp=<1000 1000 1000>
    componentLinearVelScale=<100 100 100>
    physicsBodyDefinitions [{
      sphereCollisionRadius=0
      boundBone {boneName=dyn_hat_leafD}
      boxExtents=<30 30 30>
      constraintSetup {
       angularTargetAxis=AxisY
       angularLimitsMax=<0 0 0>
       angularLimitsMin=<0 0 0>
       angularTarget=<0 0 0>
       linearAxesMax=<2 2 1.5>
       linearAxesMin=<-2 -1 -1.5>
      }
      localJointOffset=<0 0 0>
     }]
   }
   n36 = Anim Dynamics - Bone: dyn_hat_leafE(ComponentPose=@n35, Alpha=0.3, ExternalForce="0,0,0") {
    angularDampingOverride=0.985
    angularSpringConstant=300
    bAngularSpring=true
    bGravityOverrideInSimSpace=true
    bLinearSpring=true
    bOverrideAngularDamping=true
    bOverrideLinearDamping=true
    gravityScale=0.15
    linearDampingOverride=0.99
    linearSpringConstant=200
    alphaBoolBlend {blendOption=Linear}
    boundBone {boneName=dyn_hat_leafE}
    componentAppliedLinearAccClamp=<1000 1000 1000>
    componentLinearVelScale=<100 100 100>
    physicsBodyDefinitions [{
      sphereCollisionRadius=0
      boundBone {boneName=dyn_hat_leafE}
      boxExtents=<30 30 30>
      constraintSetup {
       angularTargetAxis=AxisY
       angularLimitsMax=<0 0 0>
       angularLimitsMin=<0 0 0>
       angularTarget=<0 0 0>
       linearAxesMax=<2 2 1.5>
       linearAxesMin=<-2 -1 -1.5>
      }
      localJointOffset=<0 0 0>
     }]
   }
   n37 = Anim Dynamics - Bone: dyn_mask_a_l(ComponentPose=@n36, Alpha=0.35, ExternalForce="0,0,0") {
    angularDampingOverride=0.985
    angularSpringConstant=200
    bAngularSpring=true
    bGravityOverrideInSimSpace=true
    bLinearSpring=true
    bOverrideAngularDamping=true
    bOverrideLinearDamping=true
    gravityScale=0.25
    linearDampingOverride=0.99
    linearSpringConstant=150
    simulationSpace=RootRelative
    alphaBoolBlend {blendOption=Linear}
    boundBone {boneName=dyn_mask_a_l}
    componentAppliedLinearAccClamp=<500 500 500>
    componentLinearVelScale=<100 100 100>
    physicsBodyDefinitions [{
      sphereCollisionRadius=0
      boundBone {boneName=dyn_mask_a_l}
      boxExtents=<20 20 20>
      constraintSetup {
       angularTargetAxis=AxisY
       angularLimitsMax=<0 0 0>
       angularLimitsMin=<0 0 0>
       angularTarget=<0 0 0>
       linearAxesMax=<3 2 4>
       linearAxesMin=<-3 -2 -3>
      }
      localJointOffset=<0 0 0>
     }]
   }
   n38 = Anim Dynamics - Bone: dyn_mask_b_l(ComponentPose=@n37, Alpha=0.275, ExternalForce="0,0,0") {
    angularDampingOverride=0.985
    angularSpringConstant=200
    bAngularSpring=true
    bGravityOverrideInSimSpace=true
    bLinearSpring=true
    bOverrideAngularDamping=true
    bOverrideLinearDamping=true
    gravityScale=0.28
    linearDampingOverride=0.99
    linearSpringConstant=150
    simulationSpace=RootRelative
    alphaBoolBlend {blendOption=Linear}
    boundBone {boneName=dyn_mask_b_l}
    componentAppliedLinearAccClamp=<500 500 500>
    componentLinearVelScale=<100 100 100>
    physicsBodyDefinitions [{
      sphereCollisionRadius=0
      boundBone {boneName=dyn_mask_b_l}
      boxExtents=<25 25 25>
      constraintSetup {
       angularTargetAxis=AxisY
       angularLimitsMax=<0 0 0>
       angularLimitsMin=<0 0 0>
       angularTarget=<0 0 0>
       linearAxesMax=<3 2 4>
       linearAxesMin=<-3 -2 -3>
      }
      localJointOffset=<0 0 0>
     }]
   }
   n39 = Anim Dynamics - Bone: dyn_mask_a_r(ComponentPose=@n38, Alpha=0.35, ExternalForce="0,0,0") {
    angularDampingOverride=0.985
    angularSpringConstant=200
    bAngularSpring=true
    bGravityOverrideInSimSpace=true
    bLinearSpring=true
    bOverrideAngularDamping=true
    bOverrideLinearDamping=true
    gravityScale=0.2
    linearDampingOverride=0.99
    linearSpringConstant=150
    simulationSpace=RootRelative
    alphaBoolBlend {blendOption=Linear}
    boundBone {boneName=dyn_mask_a_r}
    componentAppliedLinearAccClamp=<500 500 500>
    componentLinearVelScale=<100 100 100>
    physicsBodyDefinitions [{
      sphereCollisionRadius=0
      boundBone {boneName=dyn_mask_a_r}
      boxExtents=<20 20 20>
      constraintSetup {
       angularTargetAxis=AxisY
       angularLimitsMax=<0 0 0>
       angularLimitsMin=<0 0 0>
       angularTarget=<0 0 0>
       linearAxesMax=<3 2 4>
       linearAxesMin=<-3 -2 -3>
      }
      localJointOffset=<0 0 0>
     }]
   }
   n40 = Anim Dynamics - Bone: dyn_mask_b_r(ComponentPose=@n39, Alpha=0.3, ExternalForce="0,0,0") {
    angularDampingOverride=0.985
    angularSpringConstant=200
    bAngularSpring=true
    bGravityOverrideInSimSpace=true
    bLinearSpring=true
    bOverrideAngularDamping=true
    bOverrideLinearDamping=true
    gravityScale=0.28
    linearDampingOverride=0.99
    linearSpringConstant=150
    simulationSpace=RootRelative
    alphaBoolBlend {blendOption=Linear}
    boundBone {boneName=dyn_mask_b_r}
    componentAppliedLinearAccClamp=<500 500 500>
    componentLinearVelScale=<100 100 100>
    physicsBodyDefinitions [{
      sphereCollisionRadius=0
      boundBone {boneName=dyn_mask_b_r}
      boxExtents=<25 25 25>
      constraintSetup {
       angularTargetAxis=AxisY
       angularLimitsMax=<0 0 0>
       angularLimitsMin=<0 0 0>
       angularTarget=<0 0 0>
       linearAxesMax=<3 2 4>
       linearAxesMin=<-3 -2 -3>
      }
      localJointOffset=<0 0 0>
     }]
   }
   n41 = Anim Dynamics - Bone: dyn_hat_pumpA(ComponentPose=@n40, Alpha=0.62, ExternalForce="0,0,0") {
    angularDampingOverride=0.985
    angularSpringConstant=120
    bAngularSpring=true
    bGravityOverrideInSimSpace=true
    bOverrideAngularDamping=true
    gravityScale=0.355
    alphaBoolBlend {blendOption=Linear}
    boundBone {boneName=dyn_hat_pumpA}
    componentAppliedLinearAccClamp=<1000 1000 1000>
    componentLinearVelScale=<5 5 5>
    physicsBodyDefinitions [{
      sphereCollisionRadius=0
      boundBone {boneName=dyn_hat_pumpA}
      boxExtents=<45 35 45>
      constraintSetup {angularLimitsMax=<20 22 19>, angularLimitsMin=<-20 -15 -15>, angularTarget=<0 0 0>}
      localJointOffset=<0 0 -10>
     }]
   }
   n42 = Anim Dynamics - Bone: dyn_hat_pumpB(ComponentPose=@n41, Alpha=0.575, ExternalForce="0,0,0") {
    angularDampingOverride=0.985
    angularSpringConstant=120
    bAngularSpring=true
    bGravityOverrideInSimSpace=true
    bOverrideAngularDamping=true
    gravityScale=0.275
    alphaBoolBlend {blendOption=Linear}
    boundBone {boneName=dyn_hat_pumpB}
    componentAppliedLinearAccClamp=<1000 1000 1000>
    componentLinearVelScale=<5 5 5>
    physicsBodyDefinitions [{
      sphereCollisionRadius=0
      boundBone {boneName=dyn_hat_pumpB}
      boxExtents=<35 35 35>
      constraintSetup {angularLimitsMax=<15 15 15>, angularLimitsMin=<-15 -15 -15>, angularTarget=<0 0 0>}
      localJointOffset=<0 0 -8>
     }]
   }
   n43 = Anim Dynamics - Bone: dyn_hat_tipA(ComponentPose=@n42, Alpha=1, ExternalForce="0,0,0") {
    angularDampingOverride=0.985
    angularSpringConstant=100
    bAngularSpring=true
    bGravityOverrideInSimSpace=true
    bLinearSpring=true
    bOverrideAngularDamping=true
    bOverrideLinearDamping=true
    gravityScale=0.33
    linearDampingOverride=0.995
    linearSpringConstant=275
    simulationSpace=RootRelative
    alphaBoolBlend {blendOption=Linear}
    alphaScaleBiasClamp {bInterpResult=true, interpSpeedDecreasing=8, interpSpeedIncreasing=8}
    boundBone {boneName=dyn_hat_tipA}
    componentAppliedLinearAccClamp=<500 500 500>
    componentLinearAccScale=<1 1 1>
    componentLinearVelScale=<100 100 100>
    physicsBodyDefinitions [{
      sphereCollisionRadius=0
      boundBone {boneName=dyn_hat_tipA}
      boxExtents=<30 30 30>
      constraintSetup {
       angularTargetAxis=AxisY
       angularLimitsMax=<0 0 0>
       angularLimitsMin=<0 0 0>
       angularTarget=<0 0 0>
       linearAxesMax=<2.6406 3.5 4>
       linearAxesMin=<-3.0262 -3 -3>
      }
      localJointOffset=<0 0 0>
     }]
   }
   RigidBody(ComponentPose=@n43, Alpha=0.2) {
    bOverrideWorldGravity=true
    overridePhysicsAsset=/BRCosmetics/Accessories/Hats/M_MED_Commando_PumpkinMan_Hat/Meshes/M_MED_Commando_PumpkinMan_Head_Physics
    alphaBoolBlend {blendOption=Linear}
    componentAppliedLinearAccClamp=<100 100 100>
    componentLinearAccScale=<1 1 1>
    componentLinearVelScale=<150 150 150>
    overrideWorldGravity {z=-980}
   }
   RigidBody(Alpha=0) {
    bClampLinearTranslationLimitToRefPose=true
    bOverrideWorldGravity=true
    overridePhysicsAsset=&/F_MED_Desert_Ops_Swamp_FaceAcc_PhysicsAsset
    simulationSpace=BaseBoneSpace
    alphaBoolBlend {blendOption=Linear}
    alphaScaleBiasClamp {bInterpResult=true}
    baseBoneRef {boneName=root}
    componentAppliedLinearAccClamp=<2000 2000 2000>
    componentLinearAccScale=<0.7 0.7 0.7>
    componentLinearVelScale=<0.7 0.7 0.7>
    overrideWorldGravity {z=-980}
   }
   n44 = Get hair_alpha()
   n45 = Flap Node - Bone: hat_rear!(
    ComponentPose:Component Space Pose Link Structure=Local To Component(LocalPose=Copy Pose From Mesh() {bUseAttachedParent=true})
    Alpha:float=1
   ) -> (Pose:Component Space Pose Link Structure) {
    is=title FortAnimGraphNode_Flap
    bUseClamp=true
    targetClampMax=250
    targetComponent=RotationZ
    alphaBoolBlend {blendOption=Linear}
    sourceDrives [{
      bUseRange=false
      drivingCurve=null
      multiplier=2.2
      rangeMax=1
      rangeMin=0
      remappedMax=1
      remappedMin=0
      sourceComponent=RotationZ
      useQuaternion=true
      sourceBone {boneName=head}
     }]
    targetBone {boneName=hat_rear}
   }
   n46 = Flap Node - Bone: dyn_hair_bk_01!(ComponentPose:Component Space Pose Link Structure=@n45, Alpha:float=1) -> (Pose:Component Space Pose Link Structure) {
    is=title FortAnimGraphNode_Flap
    bUseClamp=true
    targetClampMax=100
    targetClampMin=-5
    targetComponent=TranslationY
    alphaBoolBlend {blendOption=Linear}
    sourceDrives [{
      bUseRange=false
      drivingCurve=null
      multiplier=6
      rangeMax=1
      rangeMin=0
      remappedMax=1
      remappedMin=0
      sourceComponent=RotationZ
      useQuaternion=true
      sourceBone {boneName=head}
     }]
    targetBone {boneName=dyn_hair_bk_01}
   }
   n47 = Flap Node - Bone: dyn_hair_bk_l_01!(ComponentPose:Component Space Pose Link Structure=@n46, Alpha:float=1) -> (Pose:Component Space Pose Link Structure) {
    is=title FortAnimGraphNode_Flap
    bUseClamp=true
    targetClampMax=100
    targetClampMin=-5
    targetComponent=TranslationY
    alphaBoolBlend {blendOption=Linear}
    sourceDrives [{
      bUseRange=false
      drivingCurve=null
      multiplier=6
      rangeMax=1
      rangeMin=0
      remappedMax=1
      remappedMin=0
      sourceComponent=RotationZ
      useQuaternion=true
      sourceBone {boneName=head}
     }]
    targetBone {boneName=dyn_hair_bk_l_01}
   }
   n48 = Flap Node - Bone: dyn_hair_bk_r_01!(ComponentPose:Component Space Pose Link Structure=@n47, Alpha:float=1) -> (Pose:Component Space Pose Link Structure) {
    is=title FortAnimGraphNode_Flap
    bUseClamp=true
    targetClampMax=100
    targetClampMin=-5
    targetComponent=TranslationY
    alphaBoolBlend {blendOption=Linear}
    sourceDrives [{
      bUseRange=false
      drivingCurve=null
      multiplier=6
      rangeMax=1
      rangeMin=0
      remappedMax=1
      remappedMin=0
      sourceComponent=RotationZ
      useQuaternion=true
      sourceBone {boneName=head}
     }]
    targetBone {boneName=dyn_hair_bk_r_01}
   }
   n49 = Flap Node - Bone: dyn_hair_bk_01(ComponentPose=@n48, Alpha=1) {
    bUseClamp=true
    targetClampMax=100
    targetClampMin=-5
    targetComponent=TranslationX
    alphaBoolBlend {blendOption=Linear}
    sourceDrives [{
      bUseRange=false
      drivingCurve=null
      multiplier=3
      rangeMax=1
      rangeMin=0
      remappedMax=1
      remappedMin=0
      sourceComponent=RotationZ
      useQuaternion=true
      sourceBone {boneName=head}
     }]
    targetBone {boneName=dyn_hair_bk_01}
   }
   n50 = Flap Node - Bone: dyn_hair_bk_l_01(ComponentPose=@n49, Alpha=1) {
    bUseClamp=true
    targetClampMax=100
    targetClampMin=-5
    targetComponent=TranslationX
    alphaBoolBlend {blendOption=Linear}
    sourceDrives [{
      bUseRange=false
      drivingCurve=null
      multiplier=3
      rangeMax=1
      rangeMin=0
      remappedMax=1
      remappedMin=0
      sourceComponent=RotationZ
      useQuaternion=true
      sourceBone {boneName=head}
     }]
    targetBone {boneName=dyn_hair_bk_l_01}
   }
   n51 = Flap Node - Bone: dyn_hair_bk_r_01(ComponentPose=@n50, Alpha=1) {
    bUseClamp=true
    targetClampMax=100
    targetClampMin=-5
    targetComponent=TranslationX
    alphaBoolBlend {blendOption=Linear}
    sourceDrives [{
      bUseRange=false
      drivingCurve=null
      multiplier=3
      rangeMax=1
      rangeMin=0
      remappedMax=1
      remappedMin=0
      sourceComponent=RotationZ
      useQuaternion=true
      sourceBone {boneName=head}
     }]
    targetBone {boneName=dyn_hair_bk_r_01}
   }
   n52 = Flap Node - Bone: dyn_hair_bk_02!(ComponentPose:Component Space Pose Link Structure=@n51, Alpha:float=1) -> (Pose:Component Space Pose Link Structure) {
    is=title FortAnimGraphNode_Flap
    bUseClamp=true
    targetClampMax=100
    targetClampMin=-5
    targetComponent=TranslationX
    alphaBoolBlend {blendOption=Linear}
    sourceDrives [{
      bUseRange=false
      drivingCurve=null
      multiplier=-3
      rangeMax=1
      rangeMin=0
      remappedMax=1
      remappedMin=0
      sourceComponent=RotationZ
      useQuaternion=true
      sourceBone {boneName=head}
     }]
    targetBone {boneName=dyn_hair_bk_02}
   }
   n53 = Flap Node - Bone: dyn_hair_bk_l_02!(ComponentPose:Component Space Pose Link Structure=@n52, Alpha:float=1) -> (Pose:Component Space Pose Link Structure) {
    is=title FortAnimGraphNode_Flap
    bUseClamp=true
    targetClampMax=100
    targetClampMin=-5
    targetComponent=TranslationX
    alphaBoolBlend {blendOption=Linear}
    sourceDrives [{
      bUseRange=false
      drivingCurve=null
      multiplier=-3
      rangeMax=1
      rangeMin=0
      remappedMax=1
      remappedMin=0
      sourceComponent=RotationZ
      useQuaternion=true
      sourceBone {boneName=head}
     }]
    targetBone {boneName=dyn_hair_bk_l_02}
   }
   n54 = Flap Node - Bone: dyn_hair_bk_r_02!(ComponentPose:Component Space Pose Link Structure=@n53, Alpha:float=1) -> (Pose:Component Space Pose Link Structure) {
    is=title FortAnimGraphNode_Flap
    bUseClamp=true
    targetClampMax=100
    targetClampMin=-5
    targetComponent=TranslationX
    alphaBoolBlend {blendOption=Linear}
    sourceDrives [{
      bUseRange=false
      drivingCurve=null
      multiplier=-3
      rangeMax=1
      rangeMin=0
      remappedMax=1
      remappedMin=0
      sourceComponent=RotationZ
      useQuaternion=true
      sourceBone {boneName=head}
     }]
    targetBone {boneName=dyn_hair_bk_r_02}
   }
   n55 = Get jacket_ext_force()
   n56 = Anim Dynamics - Bone: dyn_hair_bk_01(ComponentPose=@n54, Alpha=@n44, ExternalForce=@n55) {
    angularDampingOverride=0.99
    angularSpringConstant=1000
    bAngularSpring=true
    bGravityOverrideInSimSpace=true
    bOverrideAngularDamping=true
    bUsePlanarLimit=false
    bUseSphericalLimits=true
    gravityScale=0.35
    alphaBoolBlend {blendOption=Linear}
    boundBone {boneName=dyn_hair_bk_01}
    componentAppliedLinearAccClamp=<70 70 70>
    componentLinearVelScale=<300 300 300>
    gravityOverride {z=-980}
    physicsBodyDefinitions [{
      collisionType=CustomSphere
      sphereCollisionRadius=4.3
      boundBone {boneName=dyn_hair_bk_01}
      boxExtents=<5 30 30>
      constraintSetup {angularLimitsMax=<0 20 25>, angularLimitsMin=<0 -20 -65>, angularTarget=<0 0 0>}
      localJointOffset=<6.5 0 0>
     }]
    planarLimits [{drivingBone {boneName=neck_01}, planeTransform {location=<14 -5 6>, rotation=<rot 90 -90 0>}}]
    sphericalLimits [{limitRadius=70, drivingBone {boneName=spine_05}, sphereLocalOffset=<-30 50 0>}]
   }
   n57 = Anim Dynamics - Bone: dyn_hair_bk_02(ComponentPose=@n56, Alpha=@n44, ExternalForce=@n55) {
    angularDampingOverride=0.99
    angularSpringConstant=1000
    bAngularSpring=true
    bGravityOverrideInSimSpace=true
    bOverrideAngularDamping=true
    bUseSphericalLimits=true
    gravityScale=0.35
    alphaBoolBlend {blendOption=Linear}
    boundBone {boneName=dyn_hair_bk_02}
    componentAppliedLinearAccClamp=<70 70 70>
    componentLinearVelScale=<300 300 300>
    gravityOverride {z=-980}
    physicsBodyDefinitions [{
      collisionType=CustomSphere
      sphereCollisionRadius=4
      boundBone {boneName=dyn_hair_bk_02}
      boxExtents=<5 40 40>
      constraintSetup {angularLimitsMax=<0 20 7>, angularLimitsMin=<0 -20 -75>, angularTarget=<0 0 0>}
      localJointOffset=<10 0 0>
     }]
    planarLimits [
     {drivingBone {boneName=neck_01}, planeTransform {location=<7 -7 6>, rotation=<rot -90 -90 0>}}
     {drivingBone {boneName=spine_05}, planeTransform {location=<7 -11.25 6>, rotation=<rot -90 -36.5652 0>}}
    ]
    sphericalLimits [
     {limitRadius=45, drivingBone {boneName=spine_05}, sphereLocalOffset=<-25 18 0>}
     {limitRadius=8, drivingBone {boneName=neck_02}, sphereLocalOffset=<0 -1.5 0>}
    ]
   }
   n58 = Anim Dynamics - Bone: dyn_hair_bk_r_01(ComponentPose=@n57, Alpha=@n44, ExternalForce=@n55) {
    angularDampingOverride=0.99
    angularSpringConstant=1000
    bAngularSpring=true
    bGravityOverrideInSimSpace=true
    bOverrideAngularDamping=true
    bUsePlanarLimit=false
    gravityScale=0.35
    alphaBoolBlend {blendOption=Linear}
    boundBone {boneName=dyn_hair_bk_r_01}
    componentAppliedLinearAccClamp=<70 70 70>
    componentLinearVelScale=<300 300 300>
    gravityOverride {z=-980}
    physicsBodyDefinitions [{
      collisionType=CustomSphere
      sphereCollisionRadius=3
      boundBone {boneName=dyn_hair_bk_r_01}
      boxExtents=<5 30 30>
      constraintSetup {angularLimitsMax=<0 10 20>, angularLimitsMin=<0 -10 -65>, angularTarget=<0 0 0>}
      localJointOffset=<6.5 0 0>
     }]
    sphericalLimits [{limitRadius=7, drivingBone {boneName=neck_02}, sphereLocalOffset=<0 0 0>}]
   }
   n59 = Anim Dynamics - Bone: dyn_hair_bk_r_02(ComponentPose=@n58, Alpha=@n44, ExternalForce=@n55) {
    angularDampingOverride=0.99
    angularSpringConstant=1000
    bAngularSpring=true
    bGravityOverrideInSimSpace=true
    bOverrideAngularDamping=true
    bUseSphericalLimits=true
    gravityScale=0.35
    alphaBoolBlend {blendOption=Linear}
    boundBone {boneName=dyn_hair_bk_r_02}
    componentAppliedLinearAccClamp=<70 70 70>
    componentLinearVelScale=<300 300 300>
    gravityOverride {z=-980}
    physicsBodyDefinitions [{
      collisionType=CustomSphere
      sphereCollisionRadius=4
      boundBone {boneName=dyn_hair_bk_r_02}
      boxExtents=<5 40 40>
      constraintSetup {angularLimitsMax=<0 10 7>, angularLimitsMin=<0 -10 -75>, angularTarget=<0 0 0>}
      localJointOffset=<10 0 0>
     }]
    sphericalLimits [{limitRadius=55.125, drivingBone {boneName=spine_05}, sphereLocalOffset=<-40 23 8>}]
   }
   n60 = Anim Dynamics - Bone: dyn_hair_bk_l_01(ComponentPose=@n59, Alpha=@n44, ExternalForce=@n55) {
    angularDampingOverride=0.99
    angularSpringConstant=1000
    bAngularSpring=true
    bGravityOverrideInSimSpace=true
    bOverrideAngularDamping=true
    bUsePlanarLimit=false
    gravityScale=0.35
    alphaBoolBlend {blendOption=Linear}
    boundBone {boneName=dyn_hair_bk_l_01}
    componentAppliedLinearAccClamp=<70 70 70>
    componentLinearVelScale=<300 300 300>
    gravityOverride {z=-980}
    physicsBodyDefinitions [{
      collisionType=CustomSphere
      sphereCollisionRadius=3
      boundBone {boneName=dyn_hair_bk_l_01}
      boxExtents=<5 30 30>
      constraintSetup {angularLimitsMax=<0 10 20>, angularLimitsMin=<0 -10 -65>, angularTarget=<0 0 0>}
      localJointOffset=<6.5 0 0>
     }]
    sphericalLimits [{limitRadius=7.5, drivingBone {boneName=neck_02}, sphereLocalOffset=<0 0 0>}]
   }
   n61 = Anim Dynamics - Bone: dyn_hair_bk_l_02(ComponentPose=@n60, Alpha=@n44, ExternalForce=@n55) {
    angularDampingOverride=0.99
    angularSpringConstant=1000
    bAngularSpring=true
    bGravityOverrideInSimSpace=true
    bOverrideAngularDamping=true
    bUsePlanarLimit=false
    bUseSphericalLimits=true
    gravityScale=0.35
    alphaBoolBlend {blendOption=Linear}
    boundBone {boneName=dyn_hair_bk_l_02}
    componentAppliedLinearAccClamp=<70 70 70>
    componentLinearVelScale=<300 300 300>
    gravityOverride {z=-980}
    physicsBodyDefinitions [{
      collisionType=CustomSphere
      sphereCollisionRadius=4
      boundBone {boneName=dyn_hair_bk_l_02}
      boxExtents=<5 40 40>
      constraintSetup {angularLimitsMax=<0 10 7>, angularLimitsMin=<0 -10 -75>, angularTarget=<0 0 0>}
      localJointOffset=<10 0 0>
     }]
    sphericalLimits [{limitRadius=46, drivingBone {boneName=spine_05}, sphereLocalOffset=<-30 19 -8>}]
   }
   n62 = Anim Dynamics - Bone: dyn_hair_l_01(ComponentPose=@n61, Alpha=@n44, ExternalForce=@n55) {
    angularDampingOverride=0.99
    angularSpringConstant=1000
    bAngularSpring=true
    bGravityOverrideInSimSpace=true
    bOverrideAngularDamping=true
    bUsePlanarLimit=false
    gravityScale=0.35
    alphaBoolBlend {blendOption=Linear}
    boundBone {boneName=dyn_hair_l_01}
    componentAppliedLinearAccClamp=<70 70 70>
    componentLinearVelScale=<300 300 300>
    gravityOverride {z=-980}
    physicsBodyDefinitions [{
      collisionType=CustomSphere
      sphereCollisionRadius=3
      boundBone {boneName=dyn_hair_l_01}
      boxExtents=<5 30 30>
      constraintSetup {angularLimitsMax=<0 8 3>, angularLimitsMin=<0 -8 -11>, angularTarget=<0 0 0>}
      localJointOffset=<-6.5 0 0>
     }]
    sphericalLimits [
     {limitRadius=8, drivingBone {boneName=neck_02}, sphereLocalOffset=<0 0 0>}
     {limitRadius=60, drivingBone {boneName=spine_05}, sphereLocalOffset=<-42 5 0>}
    ]
   }
   n63 = Anim Dynamics - Bone: dyn_hair_l_02(ComponentPose=@n62, Alpha=@n44, ExternalForce=@n55) {
    angularDampingOverride=0.99
    angularSpringConstant=200
    bAngularSpring=true
    bGravityOverrideInSimSpace=true
    bOverrideAngularDamping=true
    bUsePlanarLimit=false
    bUseSphericalLimits=true
    gravityScale=0.35
    alphaBoolBlend {blendOption=Linear}
    boundBone {boneName=dyn_hair_l_02}
    componentAppliedLinearAccClamp=<70 70 70>
    componentLinearVelScale=<300 300 300>
    gravityOverride {z=-980}
    physicsBodyDefinitions [{
      collisionType=CustomSphere
      sphereCollisionRadius=2.25
      boundBone {boneName=dyn_hair_l_02}
      boxExtents=<10 30 30>
      constraintSetup {angularLimitsMax=<0 8 40>, angularLimitsMin=<0 -8 -15>, angularTarget=<0 0 0>}
      localJointOffset=<-6.5 0 0>
     }]
    planarLimits [{drivingBone {boneName=clavicle_l}, planeTransform {location=<0 -2 0>, rotation=<rot -90 90 0>}}]
    sphericalLimits [
     {limitRadius=59, drivingBone {boneName=spine_05}, sphereLocalOffset=<-48 -6.5 0>}
     {limitRadius=6, drivingBone {boneName=spine_05}, sphereLocalOffset=<13 0 0>}
    ]
   }
   n64 = Anim Dynamics - Bone: dyn_hair_l_03(ComponentPose=@n63, Alpha=@n44, ExternalForce=@n55) {
    angularDampingOverride=0.99
    angularSpringConstant=1000
    bAngularSpring=true
    bGravityOverrideInSimSpace=true
    bOverrideAngularDamping=true
    bUsePlanarLimit=false
    bUseSphericalLimits=true
    gravityScale=0.35
    alphaBoolBlend {blendOption=Linear}
    boundBone {boneName=dyn_hair_l_03}
    componentAppliedLinearAccClamp=<70 70 70>
    componentLinearVelScale=<300 300 300>
    gravityOverride {z=-980}
    physicsBodyDefinitions [{
      collisionType=CustomSphere
      sphereCollisionRadius=3
      boundBone {boneName=dyn_hair_l_03}
      boxExtents=<5 30 30>
      constraintSetup {angularLimitsMax=<0 10 45>, angularLimitsMin=<0 -10 -25>, angularTarget=<0 0 0>}
      localJointOffset=<-6.5 0 0>
     }]
    sphericalLimits [
     {limitRadius=60, drivingBone {boneName=spine_05}, sphereLocalOffset=<-48 -6.5 0>}
     {limitRadius=8, drivingBone {boneName=spine_05}, sphereLocalOffset=<13 0 0>}
    ]
   }
   n65 = Anim Dynamics - Bone: dyn_hair_r_01(ComponentPose=@n64, Alpha=@n44, ExternalForce=@n55) {
    angularDampingOverride=0.99
    angularSpringConstant=1000
    bAngularSpring=true
    bGravityOverrideInSimSpace=true
    bOverrideAngularDamping=true
    bUsePlanarLimit=false
    gravityScale=0.35
    alphaBoolBlend {blendOption=Linear}
    boundBone {boneName=dyn_hair_r_01}
    componentAppliedLinearAccClamp=<70 70 70>
    componentLinearVelScale=<300 300 300>
    gravityOverride {z=-980}
    physicsBodyDefinitions [{
      collisionType=CustomSphere
      sphereCollisionRadius=3
      boundBone {boneName=dyn_hair_r_01}
      boxExtents=<5 30 30>
      constraintSetup {angularLimitsMax=<0 20 17>, angularLimitsMin=<0 -2.5 -8>, angularTarget=<0 0 0>}
      localJointOffset=<-6.5 0 0>
     }]
    sphericalLimits [
     {limitRadius=8, drivingBone {boneName=neck_02}, sphereLocalOffset=<0 0 0>}
     {limitRadius=8, drivingBone {boneName=neck_01}, sphereLocalOffset=<5 10 0>}
    ]
   }
   n66 = Anim Dynamics - Bone: dyn_hair_r_02(ComponentPose=@n65, Alpha=@n44, ExternalForce=@n55) {
    angularDampingOverride=0.99
    angularSpringConstant=200
    bAngularSpring=true
    bGravityOverrideInSimSpace=true
    bOverrideAngularDamping=true
    bUsePlanarLimit=false
    bUseSphericalLimits=true
    gravityScale=0.35
    alphaBoolBlend {blendOption=Linear}
    boundBone {boneName=dyn_hair_r_02}
    componentAppliedLinearAccClamp=<70 70 70>
    componentLinearVelScale=<300 300 300>
    gravityOverride {z=-980}
    physicsBodyDefinitions [{
      collisionType=CustomSphere
      sphereCollisionRadius=2.25
      boundBone {boneName=dyn_hair_r_02}
      boxExtents=<10 30 30>
      constraintSetup {angularLimitsMax=<0 20 20>, angularLimitsMin=<0 -5 -20>, angularTarget=<0 0 0>}
      localJointOffset=<-6.5 0 0>
     }]
    planarLimits [{drivingBone {boneName=clavicle_l}, planeTransform {location=<0 -2 0>, rotation=<rot -90 90 0>}}]
    sphericalLimits [
     {limitRadius=60, drivingBone {boneName=spine_05}, sphereLocalOffset=<-48 -6.5 0>}
     {limitRadius=6.5, drivingBone {boneName=spine_05}, sphereLocalOffset=<13 0 0>}
    ]
   }
   n67 = "Transform (Modify) Bone - Bone: dyn_hair_r_02"(ComponentPose=@n66, Alpha=1, Translation="0,0,1", Rotation="-5,-5,-8", Scale="1,1,1") {
    rotationMode=BMM_Additive
    rotationSpace=BCS_BoneSpace
    translationMode=BMM_Additive
    alphaBoolBlend {blendOption=Linear}
    boneToModify {boneName=dyn_hair_r_02}
    rotation=<rot -5 -5 -8>
    translation {z=1}
   }
   n68 = Get isTargeting()
   n69 = "Transform (Modify) Bone - Bone: hat_rear"(ComponentPose=@n67, Alpha=@n68, Translation="0,0,-1", Rotation="0,15,17", Scale="1,1,1") {
    rotationMode=BMM_Additive
    rotationSpace=BCS_BoneSpace
    translationMode=BMM_Additive
    translationSpace=BCS_BoneSpace
    alphaBoolBlend {blendOption=Linear}
    boneToModify {boneName=hat_rear}
    rotation {roll=17, yaw=15}
    translation {z=-1}
   }
   n70 = "Transform (Modify) Bone - Bone: dyn_hair_bk_r_02"(ComponentPose=@n69, Alpha=@n68, Translation="0,0,0", Rotation="0,0,-10", Scale="1,1,1") {
    rotationMode=BMM_Additive
    rotationSpace=BCS_BoneSpace
    translationMode=BMM_Additive
    translationSpace=BCS_BoneSpace
    alphaBoolBlend {blendOption=Linear}
    boneToModify {boneName=dyn_hair_bk_r_02}
    rotation {roll=-10}
   }
   n71 = "Constraint - dyn_hair_bk_r_04 (R)"(ComponentPose=@n70, Alpha=1, ConstraintWeights_0=1) {
    alphaBoolBlend {blendOption=Linear}
    boneToModify {boneName=dyn_hair_bk_r_04}
    constraintSetup [{transformType=Rotation, targetBone {boneName=dyn_hair_bk_r_03}}]
    constraintWeights [1]
   }
   n72 = "Transform (Modify) Bone - Bone: dyn_hair_bk_r_04"(ComponentPose=@n71, Alpha=@n68, Translation="0,0,0", Rotation="-35,30,0", Scale="1,1,1") {
    rotationMode=BMM_Additive
    rotationSpace=BCS_BoneSpace
    translationMode=BMM_Additive
    translationSpace=BCS_BoneSpace
    alphaBoolBlend {blendOption=Linear}
    boneToModify {boneName=dyn_hair_bk_r_04}
    rotation {pitch=-35, yaw=30}
   }
   n73 = "Transform (Modify) Bone - Bone: dyn_hair_bk_02"(ComponentPose=@n72, Alpha=@n68, Translation="0,0,0", Rotation="0,-10,0", Scale="1,1,1") {
    rotationMode=BMM_Additive
    rotationSpace=BCS_BoneSpace
    translationMode=BMM_Additive
    translationSpace=BCS_BoneSpace
    alphaBoolBlend {blendOption=Linear}
    boneToModify {boneName=dyn_hair_bk_02}
    rotation {yaw=-10}
   }
   Output Pose(Result=Component To Local(ComponentPose=@n73)) {layerGroup=None}
   n74 = Flap Node - Bone: dyn_hair_bk_01(Alpha=1) {
    bUseClamp=true
    targetClampMax=100
    targetComponent=RotationZ
    alphaBoolBlend {blendOption=Linear}
    sourceDrives [{
      bUseRange=false
      drivingCurve=null
      multiplier=3
      rangeMax=1
      rangeMin=0
      remappedMax=1
      remappedMin=0
      sourceComponent=RotationZ
      useQuaternion=true
      sourceBone {boneName=head}
     }]
    targetBone {boneName=dyn_hair_bk_01}
   }
   n75 = Flap Node - Bone: dyn_hair_bk_l_01(ComponentPose=@n74, Alpha=1) {
    bUseClamp=true
    targetClampMax=100
    targetComponent=RotationZ
    alphaBoolBlend {blendOption=Linear}
    sourceDrives [{
      bUseRange=false
      drivingCurve=null
      multiplier=3
      rangeMax=1
      rangeMin=0
      remappedMax=1
      remappedMin=0
      sourceComponent=RotationZ
      useQuaternion=true
      sourceBone {boneName=head}
     }]
    targetBone {boneName=dyn_hair_bk_l_01}
   }
   Flap Node - Bone: dyn_hair_bk_r_03!(ComponentPose:Component Space Pose Link Structure=@n75, Alpha:float=1) -> (Pose:Component Space Pose Link Structure) {
    is=title FortAnimGraphNode_Flap
    bUseClamp=true
    targetClampMax=100
    targetComponent=RotationZ
    alphaBoolBlend {blendOption=Linear}
    sourceDrives [{
      bUseRange=false
      drivingCurve=null
      multiplier=3
      rangeMax=1
      rangeMin=0
      remappedMax=1
      remappedMin=0
      sourceComponent=RotationZ
      useQuaternion=true
      sourceBone {boneName=head}
     }]
    targetBone {boneName=dyn_hair_bk_r_03}
   }
   n76 = Flap Node - Bone: dyn_hair_r_01!(ComponentPose:Component Space Pose Link Structure, Alpha:float=1) -> (Pose:Component Space Pose Link Structure) {
    is=title FortAnimGraphNode_Flap
    bUseClamp=true
    targetClampMax=100
    targetClampMin=-5
    targetComponent=TranslationX
    alphaBoolBlend {blendOption=Linear}
    sourceDrives [{
      bUseRange=false
      drivingCurve=null
      multiplier=-3
      rangeMax=1
      rangeMin=0
      remappedMax=1
      remappedMin=0
      sourceComponent=RotationZ
      useQuaternion=true
      sourceBone {boneName=head}
     }]
    targetBone {boneName=dyn_hair_r_01}
   }
   n77 = Flap Node - Bone: dyn_hair_r_02!(ComponentPose:Component Space Pose Link Structure=@n76, Alpha:float=1) -> (Pose:Component Space Pose Link Structure) {
    is=title FortAnimGraphNode_Flap
    bUseClamp=true
    targetClampMax=100
    targetClampMin=-5
    targetComponent=TranslationX
    alphaBoolBlend {blendOption=Linear}
    sourceDrives [{
      bUseRange=false
      drivingCurve=null
      multiplier=-3
      rangeMax=1
      rangeMin=0
      remappedMax=1
      remappedMin=0
      sourceComponent=RotationZ
      useQuaternion=true
      sourceBone {boneName=head}
     }]
    targetBone {boneName=dyn_hair_r_02}
   }
   n78 = Flap Node - Bone: dyn_hair_l_01!(ComponentPose:Component Space Pose Link Structure=@n77, Alpha:float=1) -> (Pose:Component Space Pose Link Structure) {
    is=title FortAnimGraphNode_Flap
    bUseClamp=true
    targetClampMax=100
    targetClampMin=-5
    targetComponent=TranslationX
    alphaBoolBlend {blendOption=Linear}
    sourceDrives [{
      bUseRange=false
      drivingCurve=null
      multiplier=-3
      rangeMax=1
      rangeMin=0
      remappedMax=1
      remappedMin=0
      sourceComponent=RotationZ
      useQuaternion=true
      sourceBone {boneName=head}
     }]
    targetBone {boneName=dyn_hair_l_01}
   }
   Flap Node - Bone: dyn_hair_l_02!(ComponentPose:Component Space Pose Link Structure=@n78, Alpha:float=1) -> (Pose:Component Space Pose Link Structure) {
    is=title FortAnimGraphNode_Flap
    bUseClamp=true
    targetClampMax=100
    targetClampMin=-5
    targetComponent=TranslationX
    alphaBoolBlend {blendOption=Linear}
    sourceDrives [{
      bUseRange=false
      drivingCurve=null
      multiplier=-3
      rangeMax=1
      rangeMin=0
      remappedMax=1
      remappedMin=0
      sourceComponent=RotationZ
      useQuaternion=true
      sourceBone {boneName=head}
     }]
    targetBone {boneName=dyn_hair_l_02}
   }
   n79 = Flap Node - Bone: dyn_hair_bk_02(Alpha=1) {
    bUseClamp=true
    targetClampMax=100
    targetClampMin=-5
    targetComponent=TranslationX
    alphaBoolBlend {blendOption=Linear}
    sourceDrives [{
      bUseRange=false
      drivingCurve=null
      multiplier=-6
      rangeMax=1
      rangeMin=0
      remappedMax=1
      remappedMin=0
      sourceComponent=RotationZ
      useQuaternion=true
      sourceBone {boneName=head}
     }]
    targetBone {boneName=dyn_hair_bk_02}
   }
   n80 = Flap Node - Bone: dyn_hair_bk_l_02(ComponentPose=@n79, Alpha=1) {
    bUseClamp=true
    targetClampMax=100
    targetClampMin=-5
    targetComponent=TranslationX
    alphaBoolBlend {blendOption=Linear}
    sourceDrives [{
      bUseRange=false
      drivingCurve=null
      multiplier=-6
      rangeMax=1
      rangeMin=0
      remappedMax=1
      remappedMin=0
      sourceComponent=RotationZ
      useQuaternion=true
      sourceBone {boneName=head}
     }]
    targetBone {boneName=dyn_hair_bk_l_02}
   }
   Flap Node - Bone: dyn_hair_bk_r_02(ComponentPose=@n80, Alpha=1) {
    bUseClamp=true
    targetClampMax=100
    targetClampMin=-5
    targetComponent=TranslationX
    alphaBoolBlend {blendOption=Linear}
    sourceDrives [{
      bUseRange=false
      drivingCurve=null
      multiplier=-6
      rangeMax=1
      rangeMin=0
      remappedMax=1
      remappedMin=0
      sourceComponent=RotationZ
      useQuaternion=true
      sourceBone {boneName=head}
     }]
    targetBone {boneName=dyn_hair_bk_r_02}
   }
   Get CrouchWalkADSUp_Alpha() -> (CrouchWalkADSUp_Alpha:double)
   Get Head_up_more() -> (Head_up_more:double)
   Get speed_mult() -> (speed_mult:double)
   Get Rigids2() -> (Rigids2:double)
   Get Rigids1() -> (Rigids1:double)
   Get ADS_Alpha() -> (ADS_Alpha:double)
   Get RecentlyFired_Alpha() -> (RecentlyFired_Alpha:double)
  ]}
 EventGraph {nodes [
   PrintString!(execute, InString="To String (Float)"(InDouble=0), bPrintToScreen, bPrintToLog, TextColor, Duration=0.3, Key) -> (then)
   > PrintString(InString="To String (Float)"(InDouble=Get CrouchWalkADSUp_Alpha()), Duration=0.3)
   Event BlueprintUpdateAnimation()
   > n1 = Sequence!(execute) -> (then_0, then_1, then_2:Exec)
   n2 = MapRangeClamped(Value=Get PawnSpeed2D(), InRangeA=0, InRangeB=100)
   > n3 = Branch(Condition="float <= float"!(A:double=@n2, B:double=0) -> (ReturnValue:bool) {is=title K2Node_CallFunction})
   > Set speed_mult(speed_mult:double=0.1) -> (Output_Get:double)
   > Set hair_alpha(hair_alpha:double=0.65) -> (Output_Get:double)
   Set speed_mult(execute=@n3.else, speed_mult=float * float(A=@n2, B=0.65))
   n4 = Branch(execute=@n1.then_1, Condition=Get bIsJumping())
   > Delay(Duration=0.25)
   > Set jacket_ext_force(jacket_ext_force:Vector="0,0,1600") -> (Output_Get:Vector)
   > n5 = Delay(Duration=0.375)
   Do Once(execute=@n4.else, Reset=@n5.then, Start Closed=true)
   > Set jacket_ext_force(jacket_ext_force="0, 0, 0")
   n6 = GetDeltaTransformFromRefPose(self=GetOwnerSkeletalMeshComponent(), BoneName=head, BaseName=spine_05)
   n7 = "Absolute (Float)"(A=BreakRotator(InRot=BreakTransform(InTransform=@n6).Rotation).Yaw)
   n8 = Get bIsTargeting()
   Set isTargeting(
    execute=@n1.then_2
    isTargeting:double=float * float(A=MapRangeClamped(Value=@n7, InRangeA=5, InRangeB=45), B="To Float (Boolean)"(InBool=@n8))
   ) -> (Output_Get:double)
   Branch(Condition="float > float"(A=0, B=20) {is=title K2Node_CallFunction})
   > Set ADS_Alpha(ADS_Alpha:double=1) -> (Output_Get:double)
   Branch(Condition="float > float"(A=0, B=70))
   > Set CrouchWalkADSUp_Alpha(CrouchWalkADSUp_Alpha:double=1) -> (Output_Get:double)
   n9 = Get PawnSpeed2D()
   n10 = Branch(Condition=AND Boolean(A=Get bIsCrouching(), B="float > float"(A=@n9, B=10)))
   n11 = Set Head_up_more(execute=@n10.else, Head_up_more:double=0) -> (Output_Get:double)
   > n12 = Set Rigids2(Rigids2:double=@n11) -> (Output_Get:double)
   > Set Rigids1(Rigids1:double=0) -> (Output_Get:double)
   > Set Constraint_Alpha(Constraint_Alpha:double=0) -> (Output_Get:double)
   Branch()
   > Set CrouchWalkADSUp_Alpha(CrouchWalkADSUp_Alpha=0)
   n13 = Branch(Condition=AND Boolean!(A="float > float"(A=0, B=0.3), B=Get bIsCrouching(), C:bool) -> (ReturnValue))
   > Set ADSUpWalk(ADSUpWalk:bool=true) -> (Output_Get:bool)
   Set ADSUpWalk(execute=@n13.else, ADSUpWalk=false)
   MapRangeClamped(Value=@n7, InRangeA=40, InRangeB=138)
   MapRangeClamped(Value="Absolute (Float)"(A=0), InRangeA=0, InRangeB=60)
   MapRangeClamped(Value=@n12, InRangeA=0, InRangeB=1, OutRangeA=1, OutRangeB=0)
   PrintString(InString="To String (Float)"(InDouble=0), Duration=0.3)
   PrintString(InString="To String (Float)"(InDouble=@n12), Duration=0.3)
   TryGetPawnOwner()
   MapRangeClamped(Value=0, InRangeA=0, InRangeB=40, OutRangeB=0.5)
   MapRangeClamped(Value=0, InRangeA=35, InRangeB=-20)
   AND Boolean!(A=false, B=NOT Boolean(A=Get bIsParachuteOpened())) -> (ReturnValue)
   float * float(A=0, B=0)
   Event OnBeginTargeting() {is=title K2Node_Event}
   Event OnEndTargeting() {is=title K2Node_Event}
   n14 = "float > float"(A=Get PawnSpeed2D(), B=50)
   Branch(Condition=@n14)
   Branch(Condition=@n14)
   Set ADS_Alpha(ADS_Alpha=0)
   Set Head_up_more(Head_up_more=0)
   Set arm_back_l(arm_back_l:double=0) -> (Output_Get:double)
   Set arm_side_lift_l(arm_side_lift_l:double=0) -> (Output_Get:double)
   Set Head_up_more(Head_up_more=1)
  ]}
]
variables [
 ADS_Alpha {type=number}
 ADSUpWalk {type=boolean}
 arm_back_l {type=number}
 arm_side_lift_l {type=number}
 Constraint_Alpha {type=number}
 CrouchWalkADSUp_Alpha {type=number}
 hair_alpha {type=number}
 Head_up_more {type=number}
 isTargeting {type=number}
 jacket_ext_force {type=Vector}
 RecentlyFired_Alpha {type=number}
 Rigids1 {type=number}
 Rigids2 {type=number}
 speed_mult {type=number}
]
